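# Physics asset for the RyzaRevT1 vehicle cosmetic body, giving it a simplified collision shape.
# A single box on the root bone encloses the body, offset forward along X and raised above the bone origin.
# It defines no other bodies and no constraints, so the asset has no articulated or ragdoll parts.
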PhysicsAsset /VehicleCosmetics/Bodies/RyzaRevT1/Meshes/Body_RyzaRevT1_PhysicsDM
bodies {root [{shapeType=Box, extentX=444, extentY=216, extentZ=116, shapeName=root_box, center=<34.5 0 89>}]}
constraints []
